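; Blueprint point provider for NPC patrol paths in PolishedJade, enabling itself for the Living World system at match start.
; On BeginPlay with authority, it waits for the async match-ready event, then for the initial calendar sync to complete.
; Once synced, it calls the deprecated EnablePointProvider on the Living World point provider interface; the construction script is empty.
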(event EventBeginPlay
  (Networking|SwitchHasAuthority
    (:Authority
      (Class|FortAsyncActionAuthorityMatchReady|MatchReadyAsync
        (:OnReady
          (MCPContext|InitialCalendarSyncCompleted
            (:OnSyncComplete
              (Class|FortAthenaLivingWorldPointProviderInterface|EnablePointProvider_DEPRECATED))))))))

(fn ConstructionScript)
